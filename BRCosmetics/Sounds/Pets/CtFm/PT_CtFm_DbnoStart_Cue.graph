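# Sound cue for the CtFm cosmetic pet, voicing the start of the DBNO (down but not out) state.
# A quality-level root routes all four of its inputs to one random node, which picks among four equally weighted DBNO wave variations.
# Playback uses the pets sound class and a close-range pet attenuation asset, with a slight volume boost.
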SoundCue /BRCosmetics/Sounds/Pets/CtFm/PT_CtFm_DbnoStart_Cue
root=QualityLevel_0
nodes [
 {id=QualityLevel_0, children [Random_1,Random_1,Random_1,Random_1]}
 {id=Random_1, children [WavePlayer_3,WavePlayer_4,WavePlayer_7,WavePlayer_8], properties {weights [1,1,1,1]}}
 {id=WavePlayer_3, properties {soundWaveAssetPtr=&/PT_CtFm_DBNO_01}}
 {id=WavePlayer_4, properties {soundWaveAssetPtr=&/PT_CtFm_DBNO_02}}
 {id=WavePlayer_7, properties {soundWaveAssetPtr=&/PT_CtFm_DBNO_03}}
 {id=WavePlayer_8, properties {soundWaveAssetPtr=&/PT_CtFm_DBNO_04}}
]
properties {
 attenuationSettings=&Fort_Pets_Close_Athena_Att
 maxDistance=1800.0001
 soundClassObject=&Fort_Creatures_Pets
 volumeMultiplier=1.12
}
