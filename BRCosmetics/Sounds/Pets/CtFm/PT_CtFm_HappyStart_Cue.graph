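# Sound cue for the start of the CtFm pet's happy reaction, part of the cosmetic pet sounds.
# A Random node with equal weights picks one of three Happy wave players (variations 01, 03 and 04).
# The cue uses close-range Athena pet attenuation and the creature pets sound class, with a slight volume boost.
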SoundCue /BRCosmetics/Sounds/Pets/CtFm/PT_CtFm_HappyStart_Cue
root=Random_1
nodes [
 {id=Random_1, children [WavePlayer_0,WavePlayer_2,WavePlayer_3], properties {weights [1,1,1]}}
 {id=WavePlayer_0, properties {soundWaveAssetPtr=&/PT_CtFm_Happy_01}}
 {id=WavePlayer_2, properties {soundWaveAssetPtr=&/PT_CtFm_Happy_03}}
 {id=WavePlayer_3, properties {soundWaveAssetPtr=&/PT_CtFm_Happy_04}}
]
properties {
 attenuationSettings=&Fort_Pets_Close_Athena_Att
 maxDistance=1800.0001
 soundClassObject=&Fort_Creatures_Pets
 volumeMultiplier=1.12
}
