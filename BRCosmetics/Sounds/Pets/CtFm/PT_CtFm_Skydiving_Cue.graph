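# Sound cue for the CtFm pet's skydiving sound.
# A quality-level node sends all four tiers to one random node that picks among four equally weighted skydive wave variations.
# It uses the close-range pet attenuation and creature pets sound class.
# A concurrency override limits it to one instance per owner, blocking new plays while one is active.
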SoundCue /BRCosmetics/Sounds/Pets/CtFm/PT_CtFm_Skydiving_Cue
root=QualityLevel_0
nodes [
 {id=QualityLevel_0, children [Random_1,Random_1,Random_1,Random_1]}
 {id=Random_1, children [WavePlayer_2,WavePlayer_4,WavePlayer_6,WavePlayer_7], properties {weights [1,1,1,1]}}
 {id=WavePlayer_2, properties {soundWaveAssetPtr=&/PT_CtFm_Skydive_01}}
 {id=WavePlayer_4, properties {soundWaveAssetPtr=&/PT_CtFm_Skydive_02}}
 {id=WavePlayer_6, properties {soundWaveAssetPtr=&/PT_CtFm_Skydive_03}}
 {id=WavePlayer_7, properties {soundWaveAssetPtr=&/PT_CtFm_Skydive_04}}
]
properties {
 attenuationSettings=&Fort_Pets_Close_Athena_Att
 bOverrideConcurrency=true
 maxDistance=1800.0001
 soundClassObject=&Fort_Creatures_Pets
 volumeMultiplier=0.896
 concurrencyOverrides {bLimitToOwner=true, maxCount=1, resolutionRule=PreventNew}
}
